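# Material function that tiles a texture in world space using triplanar projection, with separate tile sizes and outputs for the R, G and B channels.
# Each channel divides world position by its size and samples the YZ, XZ and XY planes.
# CheapContrast-sharpened absolute normal X and Z blend these samples.
# Outputs per channel are the full blend (XYZ), side blend (XY) and raw YZ sample, labelled Z.
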MaterialFunction /Engine/Functions/Engine_MaterialFunctions01/Texturing/WorldAlignedTexture_SeperateChannels
graph {
 properties {bExposeToLibrary=true, description=Tiles a texture in worldspace, libraryCategoriesText [Misc,Texturing]}
 nodes [
  n1 = FunctionInput(Preview=WorldPosition() {worldPositionShaderOffset=WPT_ExcludeAllShaderOffsets}.XYZ) {bUsePreviewValueAsDefault=true, inputName=WorldPosition}
  n2 = FunctionInput() {
   bUsePreviewValueAsDefault=true
   desc=Tilling
   description=Size of texture in worldspace units
   inputName="TextureSize (G)"
   sortPriority=-2
   previewValue=<64 64 64 6>
  }
  n3 = Divide(A=@n1, B=Multiply(A=Abs(None=@n2)) {constB=-1})
  n4 = FunctionInput(Preview=TextureObject() {texture=/Engine/EngineMaterials/DefaultDiffuse}) {inputName=TextureObject, inputType=FunctionInput_Texture2D, sortPriority=-7}
  n5 = TextureSample(UVs=ComponentMask(None=@n3) {b=true, g=true}, Tex=@n4)
  n6 = FunctionInput(Preview=VertexNormalWS()) {bUsePreviewValueAsDefault=true, inputName=World Space Normal, previewValue=<0 0 1 1>}
  n7 = ComponentMask(None=@n6) {r=true}
  n8 = FunctionInput() {
   bUsePreviewValueAsDefault=true
   inputName=ProjectionTransitionContrast
   inputType=FunctionInput_Scalar
   sortPriority=10
   previewValue=<1 0 0 1>
  }
  n9 = ComponentMask(None=MaterialFunctionCall(In=Abs(None=@n7), Contrast=@n8) {materialFunction=&CheapContrast}.Result) {r=true}
  n10 = LinearInterpolate(A=TextureSample(UVs=ComponentMask(None=@n3) {b=true, r=true}, Tex=@n4).G, B=@n5.G, Alpha=@n9) {desc=Float3}
  n11 = ComponentMask(None=@n3) {g=true, r=true}
  n12 = ComponentMask(None=@n6) {b=true}
  n13 = ComponentMask(None=MaterialFunctionCall(In=Abs(None=@n12), Contrast=@n8) {materialFunction=&CheapContrast}.Result) {r=true}
  FunctionOutput(None=LinearInterpolate(A=@n10, B=TextureSample(UVs=@n11, Tex=@n4).G, Alpha=@n13)) {outputName="XYZ Texture (G)", sortPriority=6}
  FunctionOutput(None=@n10) {outputName="XY Texture (G)", sortPriority=5}
  FunctionOutput(None=@n5.G) {outputName="Z Texture (G)", sortPriority=4}
  n14 = FunctionInput() {
   bUsePreviewValueAsDefault=true
   desc=Tilling
   description=Size of texture in worldspace units
   inputName="TextureSize (B)"
   sortPriority=-1
   previewValue=<64 64 64 6>
  }
  n15 = Divide(A=@n1, B=Multiply(A=Abs(None=@n14)) {constB=-1})
  n16 = TextureSample(UVs=ComponentMask(None=@n15) {b=true, g=true}, Tex=@n4)
  n17 = LinearInterpolate(A=TextureSample(UVs=ComponentMask(None=@n15) {b=true, r=true}, Tex=@n4).B, B=@n16.B, Alpha=@n9) {desc=Float3}
  n18 = ComponentMask(None=@n15) {g=true, r=true}
  FunctionOutput(None=LinearInterpolate(A=@n17, B=TextureSample(UVs=@n18, Tex=@n4).B, Alpha=@n13)) {outputName="XYZ Texture (B)", sortPriority=9}
  FunctionOutput(None=@n17) {outputName="XY Texture (B)", sortPriority=8}
  FunctionOutput(None=@n16.B) {outputName="Z Texture (B)", sortPriority=7}
  n19 = FunctionInput() {
   bUsePreviewValueAsDefault=true
   description=Size of texture in worldspace units
   inputName="TextureSize (R)"
   sortPriority=-3
   previewValue=<64 64 64 6>
  }
  n20 = Divide(A=@n1, B=Multiply(A=Abs(None=@n19)) {constB=-1})
  n21 = TextureSample(UVs=ComponentMask(None=@n20) {b=true, g=true}, Tex=@n4)
  n22 = LinearInterpolate(A=TextureSample(UVs=ComponentMask(None=@n20) {b=true, r=true}, Tex=@n4).R, B=@n21.R, Alpha=@n9) {desc=Float3}
  n23 = ComponentMask(None=@n20) {g=true, r=true}
  FunctionOutput(None=LinearInterpolate(A=@n22, B=TextureSample(UVs=@n23, Tex=@n4).R, Alpha=@n13)) {outputName="XYZ Texture (R)", sortPriority=3}
  FunctionOutput(None=@n22) {outputName="XY Texture (R)", sortPriority=1}
  FunctionOutput(None=@n21.R) {outputName="Z Texture (R)"}
 ]
}
