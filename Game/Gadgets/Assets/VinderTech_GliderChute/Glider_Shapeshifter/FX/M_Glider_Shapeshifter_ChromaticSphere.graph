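# Additive, unlit, two-sided Niagara mesh-particle material for the Shapeshifter glider's chromatic sphere effect.
# An inverted Fresnel samples a curve atlas color row, which is desaturated, HDR-tinted by particle color and optionally blended toward a Color tint.
# World-aligned caustics drive an erosion mask, multiplied by curve and particle alpha, with depth fade skipped on mobile.
# Dynamic parameters control erosion, Fresnel exponent, saturation and color blend per particle.
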Material /Game/Gadgets/Assets/VinderTech_GliderChute/Glider_Shapeshifter/FX/M_Glider_Shapeshifter_ChromaticSphere
graph {
 properties {
  bUsedWithNiagaraMeshParticles=true
  blendMode=BLEND_Additive
  pixelDepthOffsetMode=PDOM_Legacy
  shadingModel=MSM_Unlit
  twoSided=true
  usedShadingModels=Unlit
 }
 nodes [
  n1 = ParticleColor()
  n2 = DynamicParameter() {defaultValue=<rgba 1 2 0 0>, paramNames [Erode,FresExp,Saturation,ColorLerp]}
  n3 = CurveAtlasRowParameter(CurveTime=OneMinus(None=Fresnel(ExponentIn=@n2.FresExp))) {
   atlas=&AthenaCurveAtlas01
   curve=/Game/Athena/Environments/Nexus/Materials/FXshader/Curve_Color
   defaultValue=78
  }
  n4 = MaterialFunctionCall(LDR Input=Desaturation(None=Saturate(None=@n3), Fraction=@n2.Saturation), HDR Tint=@n1.RGB) {materialFunction=&DeriveHDRfromLDR}
  n5 = Multiply(A=VectorParameter() {parameterName=Color, defaultValue=<rgba 1 1 1 0>}.RGB, B=@n4.RemappedRange)
  MaterialFunctionCall(Emissive=LinearInterpolate(
    A=@n1.RGB
    B=LinearInterpolate(A=@n4.RemappedRange, B=@n5, Alpha=@n2.ColorLerp)
    Alpha=ScalarParameter() {defaultValue=1, parameterName=Use_ColorCurve}
   )) {materialFunction=&MF_FX_EmissiveAdjustments}
  n6 = MaterialFunctionCall!(
   TextureObject=TextureObjectParameter() {parameterName=Tex, texture=/Game/Global/Textures/Default/Caustics/T_Caustics_Softer}
   "TextureSize (R)"=VectorParameter() {parameterName=Size, defaultValue=<rgba 250 250 250 0>}.RGB
   "TextureSize (G)"
   "TextureSize (B)"
   WorldPosition
   World Space Normal
   ProjectionTransitionContrast
  ) {materialFunction=/Engine/Functions/Engine_MaterialFunctions01/Texturing/WorldAlignedTexture_SeperateChannels}
  n7 = VectorParameter() {parameterName=Detail, defaultValue=<rgba 1 1 0 0>}
  n8 = Multiply(A=Power(Base=@n6."XYZ Texture (R)", Exp=@n7.R), B=@n7.G)
  n9 = Multiply(A=@n6."XYZ Texture (R)", B=@n7.B)
  n10 = Multiply(
   A=Multiply(A=@n3.A, B=@n1.A)
   B=MaterialFunctionCall!(
    Dynamic Param=DynamicParameter() {defaultValue=<rgba 1 2 0 0>, paramNames [Erode,FresExp,Saturation,ColorLerp]}.Erode
    Alpha=Add(A=@n8, B=@n9)
   ) {materialFunction=/Game/Effects/Fort_Effects/Materials/_Functions/Erode_Alpha}.Result
  )
  n11 = ScalarParameter() {defaultValue=150, parameterName=Fade Distance}
  ShadingPathSwitch(Default=Multiply(A=@n10, B=DepthFade(FadeDistance=@n11)), Mobile=@n10)
 ]
}
